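# Conversation database for the Underground Rebel NPC, defining its greeting dialogue and onward conversation branches.
# A single entry point leads to a speech node that picks contextual lines from first-ever meeting, per-match conversation count and worn outfit checks, with a default fallback.
# First-meeting and outfit reactions use named priority curve rows.
# The speech node then links by tag to the shared NPC quest library and Jethro services conversations.
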ConversationDatabase /JethroLager/NPCs/Tandem/UndergroundRebel/Jethro_Conversation_UndergroundRebel
entry_points [{tag=Jethro.Conversation.Entry.UndergroundRebel, destinations [#1]}]
nodes [
 ConversationEntryPointNode_7 {
  id=#1
  class=ConversationEntryPointNode
  outputs [#2]
  properties {entryTag {tagName=Jethro.Conversation.Entry.UndergroundRebel}}
 }
 Speech_C_4 {
  id=#2
  class=Speech_C
  outputs [#3, #4]
  properties {
   nodeName=""
   generalConfig {
    defaultMessage="I hope you're making friends, we're going to need 'em."
    contextualMessages [
     {
      message="Don't go too far, I might need you."
      contextRequirements [{requirement {class=FortControllerRequirement_IsFirstEverConversationWithNPC}}]
      priority {curve {rowName=Tandem.Speech.Priority.FirstEverInteraction}}
     }
     {
      message="I hope you're making friends, we're going to need 'em."
      contextRequirements [{requirement {
         class=FortControllerRequirement_CheckNPCMatchConversationCount
         properties {nPCUniqueIdTag {tagName=AISpawnerData.Type.Tandem.Jethro.UndergroundRebel}}
        }}]
      priority {curve {rowName=None}}
     }
     {
      message="Keep your eyes open, I think things are about to get interesting."
      contextRequirements [{requirement {
         class=FortControllerRequirement_CheckNPCMatchConversationCount
         properties {nPCUniqueIdTag {tagName=AISpawnerData.Type.Tandem.Jethro.UndergroundRebel}}
        }}]
      priority {curve {rowName=None}}
     }
     {
      message=Hey.
      contextRequirements [{requirement {
         class=FortControllerRequirement_CheckNPCMatchConversationCount
         properties {nPCUniqueIdTag {tagName=AISpawnerData.Type.Tandem.Jethro.UndergroundRebel}}
        }}]
      priority {curve {rowName=None}}
     }
     {
      message=...
      contextRequirements [{requirement {
         class=FortControllerRequirement_HasCID
         properties {allowedCIds [
           &CID_A_422_Athena_Commando_M_Realm
           /BRCosmetics/Athena/Items/Cosmetics/Characters/Character_SequinPie
           /BRCosmetics/Athena/Items/Cosmetics/Characters/Character_ClawPad_Host
           /BRCosmetics/Athena/Items/Cosmetics/Characters/Character_MuffinLadle_Gas
          ]}
        }}]
      priority {curve {rowName=Tandem.Speech.Priority.OutfitReaction}}
     }
     {
      message="Just so you know, one of our new friends seems interested in eating you. I'm sure you'll work it out. "
      contextRequirements [{requirement {
         class=FortControllerRequirement_HasCID
         properties {allowedCIds [&CID_349_Athena_Commando_M_Banana]}
        }}]
      priority {curve {rowName=Tandem.Speech.Priority.OutfitReaction}}
     }
    ]
   }
  }
 }
 ConversationLinkNode_0 {id=#4, class=ConversationLinkNode, properties {remoteEntryTag {tagName=Conversation.Entry.JethroServices}}}
 ConversationLinkNode_1 {id=#3, class=ConversationLinkNode, properties {remoteEntryTag {tagName=Conversation.Entry.NPCLibrary.Quests}}}
]
